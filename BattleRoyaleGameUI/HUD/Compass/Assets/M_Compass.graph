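# Translucent UI material for the HUD compass, drawing a horizontally pannable strip from a grayscale distance-field texture.
# CompassPan offsets half-scaled UVs, and anti-aliasing thresholds produce a crisp shape over a softer halo.
# A sine of UV through ease curves fades the edges and center, shading adaptive Dark, Dim and Medium colors and masking opacity.
# Low-end switchers substitute simpler color and opacity values.
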Material /BattleRoyaleGameUI/HUD/Compass/Assets/M_Compass
graph {
 properties {blendMode=BLEND_Translucent, materialDomain=MD_UI, pixelDepthOffsetMode=PDOM_Legacy, usedShadingModels=Default Lit}
 nodes [
  n1 = MaterialFunctionCall!(BufferSource) {materialFunction=/BattleRoyaleGameUI/Shared/MF_UI_AdaptiveColors}
  n2 = ScalarParameter() {parameterName=CompassPan, sliderMax=10, sliderMin=-10}
  n3 = MaterialFunctionCall() {materialFunction=&GetUserInterfaceUV}
  n4 = Multiply(A=ComponentMask(None=@n3.Normalized UV) {r=true}) {constB=0.5}
  n5 = TextureSample(UVs=AppendVector(A=Add(A=@n2, B=@n4), B=ComponentMask(None=@n3.Normalized UV) {g=true})) {samplerType=SAMPLERTYPE_LinearGrayscale, texture=&/T_Compass}
  n6 = Saturate(None=MaterialFunctionCall(DistanceField=@n5.R, Threshold=Constant() {r=0.4}) {materialFunction=&MF_UI_AntiAliasing}.Outer)
  LinearInterpolate(
   A=@n1.Dark
   B=LinearInterpolate(
    A=@n1.Dim
    B=MaterialFunctionCall!(Default=@n1.Medium, LowEnd=@n1.Medium) {materialFunction=/BattleRoyaleGameUI/Shared/MF_UI_LowEnd_ValueSwitcher}.Result
    Alpha=NamedRerouteUsage()
   )
   Alpha=@n6
  )
  n7 = Sine(None=MaterialFunctionCall() {materialFunction=&GetUserInterfaceUV}.Normalized UV) {period=2}
  n8 = MaterialFunctionCall("Value (0-1)"=Saturate(None=ComponentMask(None=@n7) {r=true})) {materialFunction=&MF_UI_EaseCurves}
  n9 = Saturate(None=Power(Base=@n8.EaseInCubic) {constExponent=12})
  NamedRerouteDeclaration(None=Min(A=@n8.EaseOutQuint, B=OneMinus(None=@n9))) {name=ColorCurveMask, nodeColor=<rgba 0 0 0 1>}
  n10 = Multiply(A=MaterialFunctionCall(DistanceField=@n5.R, Threshold=Constant() {r=0.4}, Smoothing=Constant() {r=0.5}) {materialFunction=&MF_UI_AntiAliasing}.Outer) {constB=0.5}
  n11 = MaterialFunctionCall(Default=Constant() {r=1}, LowEnd=Constant() {r=0.75}) {materialFunction=&MF_UI_LowEnd_ValueSwitcher}
  n12 = OneMinus(None=Multiply(A=Power(Base=@n8.EaseInQuint) {constExponent=24}) {constB=1.75})
  n13 = Saturate(None=@n8.EaseOutQuint)
  Saturate(None=Multiply(A=LinearInterpolate(A=@n10, B=@n11.Result, Alpha=@n6), B=Min(A=@n12, B=@n13)))
 ]
}
