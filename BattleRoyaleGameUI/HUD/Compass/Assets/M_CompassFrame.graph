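# Translucent UI material that draws the HUD compass frame as an arrow plus decoration, with an IsArrowOnly static switch.
# Mirrored UVs sample two channels of T_CompassFrame as distance fields, producing arrow and decoration masks and shadows, though both shadow opacities are zero.
# Masks blend opacity, with low-end variants, and adaptive Dark, Dim and Medium colors.
# The MPC_Yellow scalar parameter blends toward the shared collection's Yellow.
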Material /BattleRoyaleGameUI/HUD/Compass/Assets/M_CompassFrame
graph {
 properties {blendMode=BLEND_Translucent, materialDomain=MD_UI, pixelDepthOffsetMode=PDOM_Legacy, usedShadingModels=Default Lit}
 nodes [
  n1 = MaterialFunctionCall!(
   UV=MaterialFunctionCall(Texture Scale=Constant() {r=1}) {materialFunction=&ScaleUVsByCenter}.UVs
   Mirror Y
   Swizzle UV
  ) {materialFunction=/FortUILibrary/MaterialFunctions/Basics/MF_UI_Symmetry_2x1_New}
  n2 = AppendVector(A=Constant(), B=Constant() {r=-0.05})
  n3 = MaterialFunctionCall!(
   DistanceField=TextureSample(UVs=Add(A=@n1.Symmetry, B=@n2)) {samplerType=SAMPLERTYPE_LinearColor, texture=&/T_CompassFrame}.R
   Threshold=Constant() {r=0.45}
   Smoothing=Constant() {r=0.05}
  ) {materialFunction=/BattleRoyaleGameUI/Shared/MF_UI_DFToMaskAndShadow}
  NamedRerouteDeclaration(None=@n3.Shadow) {name=ArrowShadow, nodeColor=<rgba 0 0 0 1>}
  n4 = MaterialFunctionCall(Texture Scale=AppendVector(A=Constant() {r=0.9}, B=Constant() {r=1})) {materialFunction=&ScaleUVsByCenter}
  n5 = Constant() {r=0.0012}
  n6 = Constant() {r=-0.05}
  n7 = TextureSample(UVs=Add(
    A=MaterialFunctionCall(UV=@n4.UVs) {materialFunction=&MF_UI_Symmetry_2x1_New}.Symmetry
    B=AppendVector(A=@n5, B=@n6)
   )) {samplerType=SAMPLERTYPE_LinearColor, texture=&T_CompassFrame}
  NamedRerouteDeclaration(None=MaterialFunctionCall(DistanceField=@n7.G, Threshold=Constant()) {materialFunction=&MF_UI_DFToMaskAndShadow}.Shadow) {name=DecorationShadow, nodeColor=<rgba 0 0 0 1>}
  NamedRerouteDeclaration(None=MaterialFunctionCall!(Default=Constant() {r=1}, LowEnd=Constant() {r=0.75}) {materialFunction=/BattleRoyaleGameUI/Shared/MF_UI_LowEnd_ValueSwitcher}.Result) {name=ArrowOpacity, nodeColor=<rgba 0 0 0 1>}
  NamedRerouteDeclaration(None=MaterialFunctionCall(Default=Constant() {r=0.25}, LowEnd=Constant() {r=0.121}) {materialFunction=&MF_UI_LowEnd_ValueSwitcher}.Result) {name=DecorationOpacity, nodeColor=<rgba 0 0 0 1>}
  NamedRerouteDeclaration(None=Constant()) {name=ArrowShadowOpacity, nodeColor=<rgba 0 0 0 1>}
  NamedRerouteDeclaration(None=Constant()) {name=DecorShadowOpacity, nodeColor=<rgba 0 0 0 1>}
  n8 = NamedRerouteUsage()
  n9 = Multiply(A=Min(A=NamedRerouteUsage(), B=OneMinus(None=@n8)), B=NamedRerouteUsage())
  n10 = MaterialFunctionCall(DistanceField=@n7.G, Threshold=Constant() {r=0.375}) {materialFunction=&MF_UI_DFToMaskAndShadow}
  n11 = NamedRerouteUsage()
  n12 = LinearInterpolate(
   A=Multiply(A=NamedRerouteDeclaration(None=@n10.Mask) {name=DecorationMask, nodeColor=<rgba 0 0 0 1>}, B=NamedRerouteUsage())
   B=NamedRerouteUsage()
   Alpha=@n11
  )
  n13 = NamedRerouteUsage()
  n14 = OneMinus(None=NamedRerouteUsage())
  n15 = Add(A=@n9, B=Multiply(A=Min(A=@n13, B=@n14), B=NamedRerouteUsage()))
  n16 = Add(A=NamedRerouteDeclaration(None=@n3.Mask) {name=ArrowMask, nodeColor=<rgba 0 0 0 1>}, B=NamedRerouteUsage())
  StaticSwitchParameter(True=LinearInterpolate(A=@n9, B=@n12, Alpha=@n11), False=LinearInterpolate(A=Saturate(None=@n15), B=@n12, Alpha=@n16)) {parameterName=IsArrowOnly}
  n17 = MaterialFunctionCall!(BufferSource) {materialFunction=/BattleRoyaleGameUI/Shared/MF_UI_AdaptiveColors}
  n18 = LinearInterpolate(
   A=LinearInterpolate(
    A=MaterialFunctionCall(Default=@n17.Dim, LowEnd=@n17.Dim) {materialFunction=&MF_UI_LowEnd_ValueSwitcher}.Result
    B=MaterialFunctionCall(Default=@n17.Medium, LowEnd=@n17.Medium) {materialFunction=&MF_UI_LowEnd_ValueSwitcher}.Result
    Alpha=NamedRerouteUsage()
   )
   B=ComponentMask(None=CollectionParameter() {collection=/BattleRoyaleGameUI/Shared/MPC_BRGameUI, parameterName=Yellow}) {b=true, g=true, r=true}
   Alpha=ScalarParameter() {parameterName=MPC_Yellow}
  )
  StaticSwitchParameter(True=LinearInterpolate(A=@n17.Dark, B=@n18, Alpha=NamedRerouteUsage()), False=LinearInterpolate(A=@n17.Dark, B=@n18, Alpha=@n16)) {parameterName=IsArrowOnly}
 ]
}
