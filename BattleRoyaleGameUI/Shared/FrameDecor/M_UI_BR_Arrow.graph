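# Translucent UI material that draws an arrow with a soft shadow for Battle Royale frame decor.
# It samples an arrow distance-field texture, scaled about its center, and thresholds the green channel twice through MF_UI_AntiAliasing for separate fill and shadow masks.
# The fill mask blends shadow opacity into fill opacity.
# Color comes from MF_UI_AdaptiveColors, with both fill and shadow using its Dim output.
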Material /BattleRoyaleGameUI/Shared/FrameDecor/M_UI_BR_Arrow
graph {
 properties {bAutomaticallySetUsageInEditor=false, blendMode=BLEND_Translucent, materialDomain=MD_UI, usedShadingModels=Default Lit}
 nodes [
  n1 = TextureSample(UVs=MaterialFunctionCall(Texture Scale=ScalarParameter() {defaultValue=1.1, parameterName=Scale}) {materialFunction=&ScaleUVsByCenter}.UVs) {samplerType=SAMPLERTYPE_LinearColor, texture=&/T_UI_BR_Arrow128x_DF}
  n2 = MaterialFunctionCall(
   DistanceField=@n1.G
   Threshold=ScalarParameter() {defaultValue=0.75, parameterName=ShadowThreshold}
   Smoothing=ScalarParameter() {defaultValue=0.5, parameterName=ShadowSmoothing}
  ) {materialFunction=&MF_UI_AntiAliasing}
  n3 = Saturate(None=MaterialFunctionCall(
    DistanceField=@n1.G
    Threshold=ScalarParameter() {defaultValue=0.75, parameterName=FillThreshold}
    Smoothing=ScalarParameter() {parameterName=FillSmoothing}
   ) {materialFunction=&MF_UI_AntiAliasing}.Outer)
  LinearInterpolate(
   A=Multiply(A=Saturate(None=@n2.Outer), B=ScalarParameter() {parameterName=ShadowOpacity})
   B=ScalarParameter() {parameterName=FillOpacity}
   Alpha=@n3
  )
  n4 = MaterialFunctionCall!(BufferSource) {materialFunction=/BattleRoyaleGameUI/Shared/MF_UI_AdaptiveColors}
  LinearInterpolate(A=@n4.Dim, B=@n4.Dim, Alpha=@n3)
 ]
}
