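# Battle Royale HUD compass widget derived from AthenaCompassBase, layering material-driven compass, frame and arrow images in an overlay.
# It also acts as a player-specific widget slotting host, registering under the Compass.Behind gameplay tag on Construct and unregistering on Destruct.
# Content sent with an exactly matching tag fills an overlay behind the compass, and removal requests detach it from that overlay.
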WidgetBlueprint /BattleRoyaleGameUI/HUD/Compass/WBP_BRGameUI_Compass
graphs [
 EventGraph {nodes [
   Event Construct()
   > RegisterWidgetHost!(
    execute
    WidgetHost:Widget Object Reference
    WidgetHostTag:"Gameplay Tag Structure (by ref)"=Get BehindCompassWidgetSlotTag() -> (BehindCompassWidgetSlotTag:Gameplay Tag Structure)
    IsPlayerSpecific:bool=true
   ) -> (then)
   Event Destruct()
   > UnregisterWidgetHost!(execute, WidgetHost:Widget Object Reference, WidgetHostTag:"Gameplay Tag Structure (by ref)"=Get BehindCompassWidgetSlotTag()) -> (then)
   n1 = Event Add Content To Host Widget!() -> (
    OutputDelegate
    then
    WidgetHostTag:"Gameplay Tag Structure (by ref)"
    ContentWidget:Widget Object Reference
    Padding:"Margin Structure (by ref)"
    bFill:bool
   ) {is=title K2Node_Event}
   > Branch(Condition=MatchesTag!(TagOne:Gameplay Tag Structure=@n1.WidgetHostTag, TagTwo:Gameplay Tag Structure=Get BehindCompassWidgetSlotTag(), bExactMatch:bool=true) -> (ReturnValue:bool))
   > n2 = AddChildToOverlay!(
    execute
    self:Overlay Object Reference=Get WidgetSlottingHostOverlay_BehindCompass() -> (WidgetSlottingHostOverlay_BehindCompass:Widget Slotting Host Overlay Object Reference)
    Content:Widget Object Reference=@n1.ContentWidget
   ) -> (then, ReturnValue:Overlay Slot Object Reference)
   > SetHorizontalAlignment(self=@n2, InHorizontalAlignment=HAlign_Fill)
   > SetVerticalAlignment(self=@n2, InVerticalAlignment=VAlign_Fill)
  ]}
 RemoveContentFromWidgetHost {nodes [
   n1 = Remove Content From Host Widget!() -> (
    then
    WidgetHostTag:"Gameplay Tag Structure (by ref)"
    ContentWidget:Widget Object Reference
    RemovalType:EWidgetSlottingRemovalType Enum=Unparent
   ) {is=title K2Node_FunctionEntry}
   > n2 = Branch(Condition=MatchesTag(TagOne=@n1, TagTwo=Get BehindCompassWidgetSlotTag(), bExactMatch=true))
   > n3 = RemoveChild(self=Get WidgetSlottingHostOverlay_BehindCompass(), Content=@n1.ContentWidget)
   > Return Node!(execute, ReturnValue:bool=@n3)
   Return Node(execute=@n2.else, ReturnValue=false)
  ]}
]
variables [BehindCompassWidgetSlotTag {type=GameplayTag, default {tagName=DynamicUI.HUD.Fort.Athena.Compass.Behind}}]
widgets {
 info {inheritedWidgetCount=2, parentClass=/Script/FortniteUI.AthenaCompassBase, rootWidgetClass=&Overlay, widgetCount=6}
 widgets [
  Overlay_0 {widgetClassPath=&Overlay, widget=Overlay_0}
  WidgetSlottingHostOverlay_BehindCompass {
   widgetClassPath=/Script/WidgetSlotting.WidgetSlottingHostOverlay
   bIsVariable=true
   parent=Overlay_0
   widget=WidgetSlottingHostOverlay_BehindCompass
   properties {hostTag {tagName=DynamicUI.HUD.Fort.Athena.Compass.Behind}}
   slot {class=OverlaySlot, properties {horizontalAlignment=HAlign_Fill, verticalAlignment=VAlign_Fill}}
  }
  Image_Compass {
   widgetClassPath=&Image
   bInherited=true
   bIsVariable=true
   parent=Overlay_0
   widget=Image_Compass
   properties {brush {
     bIsDynamicallyLoaded=false
     drawAs=Image
     imageType=FullColor
     mirroring=NoMirror
     resourceName=None
     resourceObject=/BattleRoyaleGameUI/HUD/Compass/Assets/M_Compass
     tiling=NoTile
     imageSize=<640 40>
     margin=<ltrb 0 0 0 0>
     outlineSettings {
      bUseBrushTransparency=false
      roundingType=HalfHeightRadius
      width=0
      color {colorUseRule=UseColor_Specified, specifiedColor=<rgba 0 0 0 0>}
      cornerRadii=<0 0 0 0>
     }
     tintColor {colorUseRule=UseColor_Specified, specifiedColor=<rgba 1 1 1 1>}
     uVRegion {bIsValid=false, max=<0 0>, min=<0 0>}
    }}
   slot {class=OverlaySlot, properties {horizontalAlignment=HAlign_Center, padding=<ltrb 0 18 0 0>}}
  }
  Image_CompassFrame {
   widgetClassPath=&Image
   parent=Overlay_0
   widget=Image_CompassFrame
   properties {
    brush {
     bIsDynamicallyLoaded=false
     drawAs=Image
     imageType=FullColor
     mirroring=NoMirror
     resourceName=None
     resourceObject=/BattleRoyaleGameUI/HUD/Compass/Assets/M_CompassFrame
     tiling=NoTile
     imageSize=<512 32>
     margin=<ltrb 0 0 0 0>
     outlineSettings {
      bUseBrushTransparency=false
      roundingType=HalfHeightRadius
      width=0
      color {colorUseRule=UseColor_Specified, specifiedColor=<rgba 0 0 0 0>}
      cornerRadii=<0 0 0 0>
     }
     tintColor {colorUseRule=UseColor_Specified, specifiedColor=<rgba 1 1 1 1>}
     uVRegion {bIsValid=false, max=<0 0>, min=<0 0>}
    }
    renderTransform {angle=0, scale=<1 1>, shear=<0 0>, translation=<0 10>}
   }
   slot {class=OverlaySlot, properties {horizontalAlignment=HAlign_Center, verticalAlignment=VAlign_Bottom}}
  }
  Image_Arrow {
   widgetClassPath=&Image
   parent=Overlay_0
   widget=Image_Arrow
   properties {
    visibility=SelfHitTestInvisible
    brush {
     bIsDynamicallyLoaded=false
     drawAs=Image
     imageType=FullColor
     mirroring=NoMirror
     resourceName=None
     resourceObject=/BattleRoyaleGameUI/Shared/FrameDecor/M_UI_BR_Arrow
     tiling=NoTile
     imageSize=<48 44>
     margin=<ltrb 0 0 0 0>
     outlineSettings {
      bUseBrushTransparency=false
      roundingType=HalfHeightRadius
      width=0
      color {colorUseRule=UseColor_Specified, specifiedColor=<rgba 0 0 0 0>}
      cornerRadii=<0 0 0 0>
     }
     tintColor {colorUseRule=UseColor_Specified, specifiedColor=<rgba 1 1 1 1>}
     uVRegion {bIsValid=false, max=<0 0>, min=<0 0>}
    }
    renderTransform {angle=0, scale=<1 1>, shear=<0 0>, translation=<0 16>}
   }
   slot {class=OverlaySlot, properties {horizontalAlignment=HAlign_Center, verticalAlignment=VAlign_Bottom}}
  }
  TutorialHighlight {widgetClassPath=&TutorialHighlight}
 ]
}
